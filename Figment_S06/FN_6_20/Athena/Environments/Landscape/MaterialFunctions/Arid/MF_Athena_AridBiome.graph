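# Arid biome material function for the Athena landscape, blending rock and sand into base color, colors-only and full material outputs.
# Two rock functions mix by the Arid_RockDetail layer under a sediment gradient, while sand takes world-projected UVs and color tints.
# A HeightLerp driven by vertex normal Z and inverted rock height blends them, unless UseAridRockBlend leaves sand alone.
# Quality and shading-path switches simplify base color rock and sand UVs for low-end and mobile.
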MaterialFunction /Figment_S06/FN_6_20/Athena/Environments/Landscape/MaterialFunctions/Arid/MF_Athena_AridBiome
graph {nodes [
  n1 = ScalarParameter() {defaultValue=10, group=Arid_Rock, parameterName=Arid_Rock_ProjContrast}
  n2 = MaterialFunctionCall!(ProjContrast=@n1) {materialFunction=/FigmentEnvironment/FN_6_20/Athena/Environments/Landscape/MaterialFunctions/Arid/MF_Athena_Arid_Rock}
  n3 = MaterialFunctionCall!(ProjContrast=@n1) {materialFunction=&/MF_Athena_Arid_Rock_02}
  n4 = LandscapeLayerSample() {parameterName=Arid_RockDetail, previewWeight=1}
  n5 = MaterialFunctionCall() {materialFunction=/FigmentEnvironment/FN_6_20/Athena/Environments/Landscape/MaterialFunctions/MF_Athena_SedimentGradient}
  n6 = Multiply(A=LinearInterpolate(A=@n2.BaseColor_Low, B=@n3.BaseColor_Low, Alpha=@n4), B=@n5.Result)
  n7 = Multiply(
   A=QualitySwitch(Default=@n6, Low=Multiply(A=@n2.BaseColor_Low, B=@n5.Result), Medium=@n6)
   B=ScalarParameter() {defaultValue=1.5, parameterName=LowMed_RockBrightness}
  )
  n8 = VectorParameter() {
   desc=Replacement adjustment for mobile without Sediment/Spec/Normal on PC
   parameterName=SedimentMobileReplace
   defaultValue=<rgba 1 0.7993 0.7031 0>
  }
  n9 = ShadingPathSwitch(Default=@n7, Deferred=@n7, Mobile=Multiply(A=@n2.BaseColor_Low, B=@n8.RGB)) {desc=Migrated from FeatureLevelSwitch}
  n10 = ScalarParameter() {defaultValue=-1.5, group=Arid_SandRockBlend, parameterName=Arid_SandRock_LOW}
  n11 = ScalarParameter() {defaultValue=1.5, group=Arid_SandRockBlend, parameterName=Arid_SandRock_High}
  n12 = ComponentMask(None=VertexNormalWS()) {b=true}
  n13 = Clamp(None=Add(
    A=LinearInterpolate(A=@n10, B=@n11, Alpha=@n12)
    B=ScalarParameter() {defaultValue=-0.5429, group=Arid_SandRockBlend, parameterName=Arid_SandRock_Offset}
   ))
  n14 = OneMinus(None=@n2.Height)
  n15 = ScalarParameter() {defaultValue=3, parameterName=Desert_SandRock_Contrast}
  n16 = MaterialFunctionCall(Transition Phase=@n13, Height Texture=@n14, Contrast=@n15) {desc=Base Color High end, materialFunction=&HeightLerp}
  n17 = Abs(None=ScalarParameter() {defaultValue=1500, group=MasterControl, parameterName=Texturesize_Z_Sand})
  n18 = ComponentMask(None=Divide(A=WorldPosition() {worldPositionShaderOffset=WPT_ExcludeAllShaderOffsets}.XYZ, B=Multiply(A=@n17) {constB=-1})) {g=true, r=true}
  n19 = LandscapeLayerCoords() {mappingScale=5}
  n20 = MaterialFunctionCall!(
   UseRoughness=StaticBool()
   UVs=ShadingPathSwitch(Default=QualitySwitch(Default=@n18, Low=@n19, High=@n18, Medium=@n18), Deferred=@n18, Mobile=@n19) {desc=Migrated from FeatureLevelSwitch}
  ) {materialFunction=/FigmentEnvironment/FN_6_20/Athena/Environments/Landscape/MaterialFunctions/Arid/MF_Athena_Arid_Sand}
  FunctionOutput(None=StaticSwitchParameter(True=LinearInterpolate(A=@n9, B=@n20.BaseColor_Plain, Alpha=@n16.Alpha) {desc=Base Color only}, False=@n20.BaseColor_Plain) {defaultValue=true, parameterName=UseAridRockBlend}) {outputName=BaseColor, sortPriority=1}
  n21 = Constant3Vector() {constant=<rgba 1 0.7455 0.5781 0>}
  n22 = MaterialFunctionCall() {materialFunction=/FigmentEnvironment/FN_6_20/Athena/Environments/Landscape/MaterialFunctions/MF_Athena_ZSandColor}
  n23 = Multiply(
   A=@n22.Result
   B=ComponentMask(None=CollectionParameter() {collection=/FigmentEnvironment/FN_6_20/Athena/Environments/Landscape/MPC/MPC_Landscape, parameterName=Arid_Sand_Tint}) {b=true, g=true, r=true}
  )
  FunctionOutput(None=StaticSwitchParameter(True=LinearInterpolate(A=@n21, B=@n23, Alpha=@n16.Alpha) {desc=Colors Only}, False=@n23) {defaultValue=true, parameterName=UseAridRockBlend}) {outputName=Colors_Only, sortPriority=2}
  n24 = MaterialFunctionCall!(
   In=MaterialFunctionCall!(Base Material=@n2.Mat, Top Material=@n3.Mat, Alpha=@n4) {materialFunction=/Engine/Functions/MaterialLayerFunctions/MatLayerBlend_Standard}.Blended Material
   Color=@n5.Result
  ) {materialFunction=/FigmentEnvironment/FN_6_20/Athena/Environments/Landscape/MaterialFunctions/MF_Athena_Material_BaseMultiply}
  n25 = MaterialFunctionCall(In=@n20.Simple_Material, Color=@n22.Result) {materialFunction=&MF_Athena_Material_BaseMultiply}
  FunctionOutput(None=StaticSwitchParameter(
    True=MaterialFunctionCall(Base Material=@n24.Result, Top Material=@n25.Result, Alpha=@n16.Alpha) {desc=Full Mat, materialFunction=&MatLayerBlend_Standard}.Blended Material
    False=@n25.Result
   ) {defaultValue=true, parameterName=UseAridRockBlend}) {outputName=Material}
  FunctionOutput(None=@n16.Lerp Alpha No Contrast) {outputName=Alpha_NoContrast, sortPriority=4}
  FunctionOutput() {outputName=------------, sortPriority=3}
  FunctionOutput(None=@n13) {outputName=Rock_TransitionPhase, sortPriority=5}
  FunctionOutput(None=@n1) {outputName=ProjectionContrast, sortPriority=6}
  MaterialFunctionCall(Transition Phase=@n13, Height Texture=@n14, Contrast=@n15) {desc=Base Color High end, materialFunction=&HeightLerp}
 ]}
