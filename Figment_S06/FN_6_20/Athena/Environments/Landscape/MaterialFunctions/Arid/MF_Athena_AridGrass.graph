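# Material function that blends a tinted arid grass layer onto Athena landscape materials using an input mask.
# The grass base color is multiplied by a landscape MPC tint, and its opacity serves as height for a contrast-controlled HeightLerp.
# The standard output always uses that height blend, while the low output switches to a scaled, clamped mask on mobile or low quality.
# A third output divides the mask by that scale for color-map adjustment.
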MaterialFunction /Figment_S06/FN_6_20/Athena/Environments/Landscape/MaterialFunctions/Arid/MF_Athena_AridGrass
graph {nodes [
  n1 = FunctionInput() {inputName=Grass, inputType=FunctionInput_MaterialAttributes}
  n2 = ComponentMask(None=CollectionParameter() {collection=/FigmentEnvironment/FN_6_20/Athena/Environments/Landscape/MPC/MPC_Landscape, parameterName=Arid_Grass_Tint}) {b=true, g=true, r=true}
  n3 = FunctionInput() {inputName=Mask, inputType=FunctionInput_Scalar, sortPriority=2}
  n4 = MaterialFunctionCall(
   Transition Phase=@n3
   Height Texture=MaterialFunctionCall!(Mat attributes=@n1) {materialFunction=/Engine/Functions/Engine_MaterialFunctions02/MatLayerBlend_BreakOpacity}.Result
   Contrast=ScalarParameter() {parameterName=Arid_Grass_Contrast}
  ) {materialFunction=&HeightLerp}
  FunctionOutput(None=MaterialFunctionCall!(
    Base Material=FunctionInput() {inputName=Material_Standard, inputType=FunctionInput_MaterialAttributes}
    Top Material=MaterialFunctionCall!(In=@n1, Color=@n2) {materialFunction=/FigmentEnvironment/FN_6_20/Athena/Environments/Landscape/MaterialFunctions/MF_Athena_Material_BaseMultiply}.Result
    Alpha=@n4.Alpha
   ) {materialFunction=/Engine/Functions/MaterialLayerFunctions/MatLayerBlend_Standard}.Blended Material) {outputName=Material_Standard}
  n5 = ScalarParameter() {defaultValue=1.5, parameterName=AridGrassMob_MP}
  n6 = Clamp(None=Multiply(A=@n3, B=@n5))
  n7 = QualitySwitch(Default=@n4.Alpha, Low=@n6, Medium=@n4.Alpha) {desc="Use Height lerp for Medium, but simple lerp for Mobile/Low"}
  FunctionOutput(None=MaterialFunctionCall(
    Base Material=FunctionInput() {inputName=Material_BaseColor, inputType=FunctionInput_MaterialAttributes, sortPriority=1}
    Top Material=MaterialFunctionCall(In=@n1, Color=@n2) {materialFunction=&MF_Athena_Material_BaseMultiply}.Result
    Alpha=ShadingPathSwitch(Default=@n7, Deferred=@n7, Mobile=@n6) {desc=Migrated from FeatureLevelSwitch}
   ) {materialFunction=&MatLayerBlend_Standard}.Blended Material) {outputName=Material_Low, sortPriority=1}
  FunctionOutput(None=Clamp(None=Divide(A=@n3, B=@n5))) {outputName=ColorMap_MaskAdjustment, sortPriority=2}
 ]}
