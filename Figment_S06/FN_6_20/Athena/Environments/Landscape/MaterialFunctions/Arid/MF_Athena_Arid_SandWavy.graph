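# Material function for the arid wavy sand landscape layer in Athena, sampling color, mask and normal textures with one shared UV input.
# Base color is the color map times a tint parameter and the mask's red channel; specular is the inverted blue channel.
# The green channel supplies height, packed into Opacity.
# It outputs material attributes plus separate Height and Normal outputs, using shared world-group wrap samplers.
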MaterialFunction /Figment_S06/FN_6_20/Athena/Environments/Landscape/MaterialFunctions/Arid/MF_Athena_Arid_SandWavy
graph {nodes [
  n1 = FunctionInput(Preview=TextureCoordinate()) {bUsePreviewValueAsDefault=true, inputType=FunctionInput_Vector2}
  n2 = Multiply(
   A=TextureSample(UVs=@n1) {samplerSource=SSM_Wrap_WorldGroupSettings, texture=/Figment_S06/FN_6_20/Environments/Sets/Arid/Textures/T_desert_sand_wavy_C}.RGB
   B=VectorParameter() {group=Arid_SandWavy, parameterName=Arid_SandWavy_Tint, defaultValue=<rgba 1.25 1.0391 0.9115 0>}.RGB
  )
  n3 = TextureSample(UVs=@n1) {
   samplerSource=SSM_Wrap_WorldGroupSettings
   samplerType=SAMPLERTYPE_Masks
   texture=/Figment_S06/FN_6_20/Environments/Sets/Arid/Textures/T_desert_sand_wavy_S
  }
  n4 = TextureSample(UVs=@n1) {
   samplerSource=SSM_Wrap_WorldGroupSettings
   samplerType=SAMPLERTYPE_Normal
   texture=/Figment_S06/FN_6_20/Environments/Sets/Arid/Textures/T_desert_sand_wavy_N
  }
  FunctionOutput(None=MakeMaterialAttributes!(
    BaseColor=Multiply(A=@n2, B=@n3.R)
    Metallic
    Specular=OneMinus(None=@n3.B)
    Roughness
    Anisotropy
    EmissiveColor
    Opacity=FunctionOutput(None=@n3.G) {outputName=Height}
    OpacityMask
    Normal=FunctionOutput(None=@n4.RGB) {outputName=Normal}
    Tangent
    WorldPositionOffset
    SubsurfaceColor
    ClearCoat
    ClearCoatRoughness
    AmbientOcclusion
    Refraction
    CustomizedUVs_0
    CustomizedUVs_1
    CustomizedUVs_2
    CustomizedUVs_3
    CustomizedUVs_4
    CustomizedUVs_5
    CustomizedUVs_6
    CustomizedUVs_7
    PixelDepthOffset
    ShadingModel
    Displacement
   ))
 ]}
